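# Masked stream water material that animates normals along a flowmap or mesh-derived direction, with optional lake transition and waterfall variants.
# Riverbed pebble textures are UV-distorted to fake refraction, while height lerps and vertex colors drive the sand edge and dithered edge opacity.
# The waterfall layer pans distorted foam channels with bump offset.
# The blended attributes pass through WaterRenderToTextureMaterialAtt, along with world-space flow direction for render-to-texture.
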Material /Game/WaterAndWind/Materials/Water/Stream_Flowmaps_01
graph {
 properties {
  bDisableDepthTest=true
  bScreenSpaceReflections=true
  bTangentSpaceNormal=false
  bUseLightmapDirectionality=false
  bUseMaterialAttributes=true
  bUsedWithNanite=true
  blendMode=BLEND_Masked
  physMaterial=/Game/FortressPhysicalMaterials/Water
  pixelDepthOffsetMode=PDOM_Legacy
  subsurfaceProfile=/Game/Effects/Fort_Effects/Materials/Water/WaterProfile
  translucencyLightingMode=TLM_SurfacePerPixelLighting
  translucencyPass=MTP_BeforeDOF
  translucentSelfShadowDensityScale=4
  translucentShadowDensityScale=1
  usedShadingModels=Default Lit
 }
 nodes [
  n1 = NamedRerouteUsage()
  n2 = Constant() {r=1}
  n3 = BlendMaterialAttributes(A=NamedRerouteUsage(), B=NamedRerouteUsage(), Alpha=NamedRerouteUsage())
  n4 = NamedRerouteUsage()
  n5 = Constant()
  n6 = VertexColor()
  n7 = Multiply(A=@n6.R, B=@n6.R)
  n8 = StaticSwitchParameter(True=Multiply(A=OneMinus(None=@n7), B=Constant3Vector() {constant=<rgba 0 0 -3 0>}), False=Constant()) {desc=+100 instructions, group=Lake, parameterName=Enable Transition To Lake?}
  n9 = MaterialFunctionCall!(
   Render To Texture Pass
   WS Flow Map Dir=NamedRerouteUsage()
   Opacity=StaticSwitchParameter(True=Constant() {r=1}, False=StaticSwitchParameter(True=@n1, False=@n2) {group=Lake, parameterName=Enable Transition To Lake?}) {parameterName=UseWaterFall}
   Default Mat Att=SetMaterialAttributes!(
    MaterialAttributes=StaticSwitchParameter(True=@n3, False=@n4) {parameterName=UseWaterFall}
    Opacity Mask=NamedRerouteUsage()
    World Position Offset=StaticSwitchParameter(True=@n5, False=@n8) {parameterName=UseWaterFall}
   ) {attributeSetTypes [#1, #2]}
  ) {materialFunction=&/WaterRenderToTextureMaterialAtt}
  n10 = GetMaterialAttributes(None=@n9.Result) {attributeGetTypes [#3]}
  n11 = Constant()
  BlendMaterialAttributes(
   A=@n9.Result
   B=SetMaterialAttributes!(MaterialAttributes=@n10.MaterialAttributes, Base Color=Multiply(A=@n10.Base Color, B=@n11)) {attributeSetTypes [#3]}
   Alpha=Constant()
  )
  ComponentMask(None=TextureCoordinate()) {r=true}
  Constant() {r=0.0125}
  Constant() {r=0.0163}
  Constant() {r=1.8399}
  Constant() {r=1.0601}
  Constant() {r=1}
  Constant() {r=0.8}
  Constant() {r=1}
  Constant() {r=1}
  n12 = TextureCoordinate()
  n13 = ComponentMask(None=VectorParameter() {group=WaterFall, parameterName=EdgeScale_Large, defaultValue=<rgba 0.5 3 1 0>}.RGB) {g=true, r=true}
  n14 = NamedRerouteUsage()
  n15 = ScalarParameter() {defaultValue=1, group=WaterFall, parameterName=EdgeScale-LargeSpeed}
  n16 = TextureSample(UVs=Panner(Coordinate=Multiply(A=@n12, B=@n13), Time=Multiply(A=@n14, B=@n15)) {speedX=2, speedY=0.05}) {samplerSource=SSM_Wrap_WorldGroupSettings, samplerType=SAMPLERTYPE_LinearColor, texture=/Game/Global/Textures/Default/Noise/T_Noise01}
  n17 = TextureCoordinate()
  n18 = ComponentMask(None=VectorParameter() {group=WaterFall, parameterName=EdgeScale_Detail, defaultValue=<rgba 1 22 1 0>}.RGB) {g=true, r=true}
  n19 = NamedRerouteUsage()
  n20 = ScalarParameter() {defaultValue=1, group=WaterFall, parameterName=EdgeScale-DetailSpeed}
  n21 = TextureSample(UVs=Panner(Coordinate=Multiply(A=@n17, B=@n18), Time=Multiply(A=@n19, B=@n20)) {speedX=1.333, speedY=-0.5}) {samplerSource=SSM_Wrap_WorldGroupSettings, samplerType=SAMPLERTYPE_LinearColor, texture=&T_Noise01}
  n22 = Panner(
   Coordinate=Multiply(A=TextureCoordinate(), B=ScalarParameter() {defaultValue=0.2, group=WaterFall, parameterName=EdgeScale_Mask})
   Time=Multiply(A=NamedRerouteUsage(), B=ScalarParameter() {defaultValue=1, group=WaterFall, parameterName=EdgeScale-MaskSpeed})
  ) {speedX=0.02, speedY=0.1}
  n23 = Saturate(None=LinearInterpolate(Alpha=TextureSample(UVs=@n22) {texture=/Game/Global/Textures/Default/Tileable_Patterns/Foam}.R) {constA=-1.7, constB=3.5})
  n24 = LinearInterpolate(
   A=Constant() {r=1}
   B=Constant()
   Alpha=MaterialFunctionCall(
    Transition Phase=VertexColor().B
    Height Texture=LinearInterpolate(A=@n16.RGB, B=@n21.RGB, Alpha=@n23)
    Contrast=ScalarParameter() {defaultValue=0.0025, group=WaterFall, parameterName=Waterfall Edge Contrast}
   ) {materialFunction=&HeightLerp}.Alpha
  )
  n25 = ScalarParameter() {defaultValue=1, group=WaterFall, parameterName=Waterfall Edge Mod}
  LinearInterpolate(
   B=Multiply(
    A=MaterialFunctionCall() {materialFunction=&DitherTemporalAA_Fortnite}.Result
    B=ScalarParameter() {defaultValue=28.2735, group=WaterFall, parameterName=WaterFall Depth jitter amount}
   ) {constB=10}
   Alpha=NamedRerouteDeclaration(None=Multiply(A=@n24, B=@n25)) {name=WaterfallMask, nodeColor=<rgba 1 0 0.9569 1>}
  ) {constB=100, desc="This was used for PDO, but it doesn't look good"}
  n26 = VertexColor()
  n27 = OneMinus(None=@n26.B)
  n28 = MaterialFunctionCall!(
   Height Offset Amount=@n27
   Texture mip bias amount=Multiply(A=@n26.B, B=ScalarParameter() {defaultValue=30, parameterName=Lake Blur amount})
   Ground texture mask=@n27
  ) {desc=Lake, materialFunction=&/ShallowWaterFunction}
  n29 = NamedRerouteUsage()
  MaterialFunctionCall!(Base Material=@n28.MA Result, Top Material=@n29, Alpha=@n26.B) {materialFunction=/Engine/Functions/MaterialLayerFunctions/MatLayerBlend_Standard}
  n30 = NamedRerouteUsage()
  n31 = NamedRerouteUsage()
  n32 = OneMinus(None=VertexColor() {desc="lake = black"}.B)
  n33 = Multiply(A=Subtract(A=NamedRerouteUsage(), B=Multiply(A=@n32) {constB=1.5}), B=ScalarParameter() {group=Water Edge, parameterName=Water Edge Mult})
  n34 = MaterialFunctionCall(
   Transition Phase=Add(A=OneMinus(None=@n31), B=ScalarParameter() {group=Water Edge, parameterName=WaterEdgeOffset})
   Height Texture=@n33
   Contrast=ScalarParameter() {group=Water Edge, parameterName=Blend Contrast}
  ) {materialFunction=&HeightLerp}
  n35 = ScalarParameter() {group=Water Edge, parameterName=WaterOpacity}
  n36 = Multiply(A=Clamp(None=Add(A=@n34.Alpha, B=@n35)), B=VertexColor().B) {constA=1}
  n37 = StaticSwitchParameter(True=BlendMaterialAttributes(A=@n28.MA Result, B=@n29, Alpha=@n26.B), False=NamedRerouteUsage()) {desc=+100 instructions, group=Lake, parameterName=Enable Transition To Lake?}
  MaterialFunctionCall(Base Material=@n37, Top Material=@n30, Alpha=@n36) {desc=Sand, materialFunction=&MatLayerBlend_Standard}
  n38 = VectorParameter() {group=WaterFall, parameterName=FoamTileCoordBase, channelNames {a="", b="", g=FoamTileR-V, r=FoamTileR-U}, defaultValue=<rgba 0.7 2.8 1 1>}
  n39 = Multiply(A=ComponentMask(None=@n38) {g=true, r=true}, B=TextureCoordinate())
  n40 = Add(
   A=OneMinus(None=MaterialFunctionCall(XY=@n39) {materialFunction=&Swizzle}.YX)
   B=NamedRerouteUsage()
  )
  n41 = NamedRerouteUsage()
  n42 = Add(
   A=Multiply(
    A=BumpOffset(Coordinate=@n40, Height=@n41) {heightRatio=0.125}
    B=ScalarParameter() {defaultValue=1.0601, group=WaterFall, parameterName=FoamTile_RChannel}
   )
   B=Multiply(A=NamedRerouteUsage(), B=ScalarParameter() {defaultValue=1, group=WaterFall, parameterName=FoamTile_RChannel-Speed})
  )
  n43 = TextureSample(UVs=@n42) {
   samplerSource=SSM_Wrap_WorldGroupSettings
   samplerType=SAMPLERTYPE_Normal
   texture=/Game/Packages/Fortnite_UniqueMaterials_1/Textures/T_WaterFall_Tile_N
  }
  Multiply(A=@n43.RGB, B=VectorParameter() {parameterName=FoamTileR-NormalMod, defaultValue=<rgba 1 1 1 1>}.RGB)
  n44 = TextureCoordinate() {uTiling=0.333, vTiling=2}
  n45 = Add(
   A=OneMinus(None=MaterialFunctionCall(XY=@n44) {materialFunction=&Swizzle}.YX)
   B=NamedRerouteUsage()
  )
  n46 = ScalarParameter() {defaultValue=0.8, group=WaterFall, parameterName=OffsetHeight}
  n47 = Add(
   A=Multiply(
    A=BumpOffset(Coordinate=@n45, Height=@n46) {heightRatio=0.2}
    B=ScalarParameter() {defaultValue=1, group=WaterFall, parameterName=FoamTile_BChannel}
   )
   B=Multiply(A=NamedRerouteUsage(), B=ScalarParameter() {defaultValue=1, group=WaterFall, parameterName=FoamTile_BChannel-Speed})
  )
  n48 = TextureSample(UVs=@n47) {
   samplerSource=SSM_Wrap_WorldGroupSettings
   samplerType=SAMPLERTYPE_Normal
   texture=&T_WaterFall_Tile_N
  }
  Multiply(A=VectorParameter() {parameterName=FoamTileB-NormalMod, defaultValue=<rgba 1 1 1 1>}.RGB, B=@n48.RGB)
  Multiply(A=TextureCoordinate(), B=ComponentMask(None=Constant3Vector() {constant=<rgba 2 20 0 1>}) {g=true, r=true})
  Multiply(A=TextureCoordinate(), B=ComponentMask(None=Constant3Vector() {constant=<rgba 0.5 3 0 1>}) {g=true, r=true})
  n49 = Multiply(A=NamedRerouteUsage(), B=ScalarParameter() {defaultValue=1.5, group=Edge Mask, parameterName=Alpha})
  n50 = ScalarParameter() {group=Edge Mask, parameterName=AlphaOffset}
  n51 = Multiply(A=VertexColor() {desc="Red = Opacity"}.R, B=Add(A=Add(A=@n49, B=@n50), B=@n32) {desc=Blue Erases Alpha}) {desc=Paintable Alpha}
  n52 = StaticSwitchParameter(
   True=MaterialFunctionCall(Transition Phase=@n51, Height Texture=@n33, Contrast=ScalarParameter() {defaultValue=0.5, group=Edge Mask, parameterName=AlphaContrast}) {materialFunction=&HeightLerp}.Alpha
   False=@n51
  ) {parameterName=UseEdgeHeightblend}
  n53 = MaterialFunctionCall(Alpha Threshold=@n52) {materialFunction=&DitherTemporalAA}
  NamedRerouteDeclaration(None=QualitySwitch(Default=@n53.Result, Low=@n52, High=@n53.Result, Medium=@n53.Result)) {name=OpacityMask, nodeColor=<rgba 1 0 0.9569 1>}
  n54 = Constant2Vector() {g=0.5, r=0.5}
  n55 = VertexColor()
  n56 = ComponentMask(None=Multiply(A=NamedRerouteUsage(), B=Constant3Vector() {constant=<rgba 1 -1 1 0>})) {g=true, r=true}
  n57 = Multiply(A=Add(A=Clamp(None=@n56) {maxDefault=0.3, minDefault=-0.3}) {constB=0.3}) {constB=1.5}
  NamedRerouteDeclaration(None=StaticSwitchParameter(True=LinearInterpolate(A=@n54, B=@n57, Alpha=@n55.B) {constB=0.5}, False=@n57) {desc=+100 instructions, group=Lake, parameterName=Enable Transition To Lake?}) {name=WSFlowMapDir, nodeColor=<rgba 1 0 0.9569 1>}
  n58 = VectorParameter() {group=FlowMap, parameterName=Flow, defaultValue=<rgba 1 1 0 0>}
  n59 = Multiply(A=@n58.RGB, B=NamedRerouteUsage())
  n60 = Multiply(
   A=Multiply(A=Constant2Vector() {g=1, r=-1}, B=ComponentMask(None=@n59) {g=true, r=true})
   B=ScalarParameter() {group=FlowMap, parameterName=Flow WPO Shift}
  )
  NamedRerouteDeclaration(None=QualitySwitch(Default=@n60, Low=Constant(), High=@n60)) {name=FFV, nodeColor=<rgba 1 0 0.9569 1>}
  n61 = NamedRerouteUsage()
  n62 = VectorParameter() {group=Refraction, parameterName=Refraction, defaultValue=<rgba 0.3 0.3 1 0>}
  n63 = Multiply(A=ComponentMask(None=Multiply(A=@n61, B=@n62.RGB)) {g=true, r=true}, B=NamedRerouteUsage()) {desc=Bank Mask}
  n64 = ScalarParameter() {group=Refraction, parameterName=Refraction Min}
  n65 = TextureSampleParameter2D(UVs=Add(A=NamedRerouteUsage(), B=Multiply(A=@n63, B=@n64) {constB=0.5})) {
   desc=normal sample - under water and on the shoreline
   parameterName=RiverBedTexture_1
   samplerSource=SSM_Wrap_WorldGroupSettings
   texture=/Game/Packages/DS_Fortnight/pebbles
  }
  n66 = NamedRerouteUsage()
  n67 = TextureSampleParameter2D!(
   UVs=Add(A=NamedRerouteUsage(), B=Add(A=@n66, B=@n63) {desc=Distortion}) {desc=Adds flow offset}
   Bias=ScalarParameter() {group=Refraction, parameterName=Refract Mip Bias}
   Apply View MipBias
  ) {mipValueMode=TMVM_MipBias, parameterName=RiverBedTexture_2, texture=/Game/Packages/Fortnite_UniqueMaterials_1/Textures/T_River_PebbleFoam_01}
  n68 = MaterialFunctionCall!(Alpha=NamedRerouteUsage(), Offset=ScalarParameter() {group=Refraction, parameterName=RefractionAlphaOffset}) {materialFunction=/Engine/Functions/Engine_MaterialFunctions01/Opacity/AlphaOffset}
  n69 = ScalarParameter() {defaultValue=4, group=Refraction, parameterName=Refraction Mask Mult}
  n70 = LinearInterpolate(A=@n65.RGB, B=@n67.RGB, Alpha=Saturate(None=Multiply(A=@n68.Result, B=@n69)))
  n71 = VectorParameter() {parameterName=Water High, defaultValue=<rgba 0.5 0.5 1 1>}
  n72 = Multiply(A=@n71.RGB, B=@n71.RGB)
  n73 = VectorParameter() {parameterName=Water Low, defaultValue=<rgba 0.1793 0.3574 1 0.2>}
  n74 = Multiply(A=@n73.RGB, B=@n73.RGB)
  n75 = NamedRerouteUsage()
  NamedRerouteDeclaration(None=Multiply(A=QualitySwitch(Default=@n70, Low=@n67.RGB, High=@n70, Medium=@n70), B=LinearInterpolate(A=@n72, B=@n74, Alpha=@n75))) {name=Something, nodeColor=<rgba 1 0 0.9569 1>}
  n76 = ScalarParameter() {defaultValue=1, group=Fresnel, parameterName=Opacity Min}
  n77 = ScalarParameter() {defaultValue=1, group=Fresnel, parameterName=Opacity Max}
  n78 = Fresnel()
  n79 = Add(A=@n78) {constB=0.15}
  n80 = NamedRerouteUsage()
  n81 = VectorParameter() {parameterName=Normal Intensity, defaultValue=<rgba 0.3 0.3 1 0>}
  n82 = SetMaterialAttributes!(
   MaterialAttributes
   Base Color=Multiply(A=NamedRerouteUsage(), B=LinearInterpolate(A=@n76, B=@n77, Alpha=@n78))
   Specular=Multiply(A=Min(A=@n79), B=ScalarParameter() {defaultValue=0.4, group=Spec, parameterName=Spec Water Edge})
   Roughness=Saturate(None=ScalarParameter() {group=Roughness, parameterName=Water Roughness}) {desc=DONT REMOVE}
   Normal=MaterialFunctionCall(BaseNormal=VertexNormalWS(), AdditionalNormal=Multiply(A=@n80, B=@n81.RGB)) {materialFunction=&BlendAngleCorrectedNormals}.Result
  ) {
   desc=Stream water
   attributeSetTypes [
    #3
    #4
    #5
    #6
   ]
  }
  NamedRerouteDeclaration(None=@n82) {name=Stream, nodeColor=<rgba 1 0 0.9569 1>}
  NamedRerouteDeclaration(None=BlendMaterialAttributes(A=@n37, B=@n30, Alpha=@n36)) {name="Sand&&WaterMat", nodeColor=<rgba 1 0 0.9569 1>}
  n83 = TextureCoordinate() {uTiling=0.4}
  n84 = Add(
   A=OneMinus(None=MaterialFunctionCall(XY=@n83) {materialFunction=&Swizzle}.YX)
   B=NamedRerouteUsage()
  )
  n85 = ScalarParameter() {defaultValue=1.8399, group=WaterFall, parameterName=FoamTile_GChannel}
  n86 = NamedRerouteUsage()
  n87 = ScalarParameter() {defaultValue=1, group=WaterFall, parameterName=FoamTile_GChannel-Speed}
  n88 = TextureSample(UVs=Add(A=Multiply(A=@n84, B=@n85), B=Multiply(A=@n86, B=@n87))) {samplerSource=SSM_Wrap_WorldGroupSettings, texture=/Game/Packages/Fortnite_UniqueMaterials_1/Textures/T_WaterFall_Tile}
  n89 = Multiply(A=VectorParameter() {group=WaterFall, parameterName=WaterFallColor, defaultValue=<rgba 0.2881 0.61 0.5573 1>}.RGB, B=@n88.G)
  n90 = TextureSample(UVs=@n42) {samplerSource=SSM_Wrap_WorldGroupSettings, texture=&T_WaterFall_Tile}
  n91 = LinearInterpolate(A=@n89, B=Add(A=@n89, B=@n90.R), Alpha=ScalarParameter() {defaultValue=0.3, group=WaterFall, parameterName=LerpFoamA})
  n92 = TextureSample(UVs=@n47) {samplerSource=SSM_Wrap_WorldGroupSettings, texture=&T_WaterFall_Tile}
  n93 = LinearInterpolate(A=@n91, B=Add(A=@n91, B=@n92.B), Alpha=ScalarParameter() {defaultValue=0.125, group=WaterFall, parameterName=LerpFoamB}) {constAlpha=0.4}
  n94 = Add(A=NamedRerouteUsage(), B=@n93)
  n95 = ScalarParameter() {defaultValue=0.3, group=WaterFall, parameterName=RoughnessMin}
  n96 = ScalarParameter() {defaultValue=1, group=WaterFall, parameterName=RoughnessMax}
  n97 = Multiply(A=Multiply(A=@n88.A, B=@n90.A), B=@n92.A)
  n98 = VectorParameter() {group=WaterFall, parameterName=ScaleNormal, defaultValue=<rgba 0.8 0.8 3.748 0>}
  n99 = LinearInterpolate(A=@n48.RGB, B=@n43.RGB, Alpha=NamedRerouteUsage())
  n100 = SetMaterialAttributes(
   Base Color=StaticSwitchParameter(True=Saturate(None=@n94), False=@n93) {parameterName=UseGradientOverlay}
   Specular=Constant() {r=0.5}
   Roughness=Clamp(None=LinearInterpolate(A=@n95, B=@n96, Alpha=@n97))
   Normal=Transform(None=Multiply(A=@n98.RGB, B=@n99))
  ) {
   desc=Waterfall
   attributeSetTypes [
    #3
    #4
    #5
    #6
   ]
  }
  NamedRerouteDeclaration(None=@n100) {name=WaterfallMat, nodeColor=<rgba 1 0 0.9569 1>}
  NamedRerouteDeclaration(None=Time() {bOverride_Period=true, period=9999999}) {name=TimeWithPeriod, nodeColor=<rgba 1 0 0.9569 1>}
  n101 = VectorParameter() {
   group=WaterFall_UVDistort
   parameterName=UVDistortCoordBase
   channelNames {a=UVDistortCoordB-V, b=UVDistortCoordB-U, g=UVDistortCoordA-V, r=UVDistortCoordA-U}
   defaultValue=<rgba 1 1 0.5 11>
  }
  n102 = TextureCoordinate()
  n103 = Multiply(A=ComponentMask(None=@n101) {g=true, r=true}, B=@n102)
  n104 = ScalarParameter() {defaultValue=0.7372, group=WaterFall_UVDistort, parameterName=UVDistortTilingA}
  n105 = NamedRerouteUsage()
  n106 = ScalarParameter() {defaultValue=1, group=WaterFall_UVDistort, parameterName=UVDistortA-Speed}
  n107 = TextureSample(UVs=Panner(Coordinate=Multiply(A=@n103, B=@n104), Time=Multiply(A=@n105, B=@n106)) {speedX=1.6, speedY=-0.08}) {samplerSource=SSM_Wrap_WorldGroupSettings, texture=&T_WaterFall_Tile}
  n108 = ScalarParameter() {defaultValue=0.0163, group=WaterFall_UVDistort, parameterName=UVDistortAmountA}
  n109 = Multiply(A=@n102, B=AppendVector(A=@n101.UVDistortCoordB-U, B=@n101.UVDistortCoordB-U))
  n110 = ScalarParameter() {defaultValue=0.8, group=WaterFall_UVDistort, parameterName=UVDistortTilingB}
  n111 = NamedRerouteUsage()
  n112 = ScalarParameter() {defaultValue=1, group=WaterFall_UVDistort, parameterName=UVDistortB-Speed}
  n113 = TextureSample(UVs=Panner(Coordinate=Multiply(A=@n109, B=@n110), Time=Multiply(A=@n111, B=@n112)) {speedX=1, speedY=0.125}) {samplerSource=SSM_Wrap_WorldGroupSettings, texture=&T_WaterFall_Tile}
  n114 = ScalarParameter() {defaultValue=0.0125, group=WaterFall_UVDistort, parameterName=UVDistortAmountB}
  n115 = ComponentMask(None=TextureCoordinate()) {r=true}
  n116 = Clamp(None=OneMinus(None=LinearInterpolate(Alpha=@n115) {constB=1.5}))
  NamedRerouteDeclaration(None=LinearInterpolate(
    A=Multiply(A=@n107.A, B=@n108)
    B=Multiply(A=@n113.A, B=@n114)
    Alpha=NamedRerouteDeclaration(None=@n116) {name=WaterfallGrad, nodeColor=<rgba 1 0 0.9569 1>}
   )) {name=UVDist1, nodeColor=<rgba 1 0 0.9569 1>}
  NamedRerouteDeclaration(None=@n113.A) {name=UVDist1Bump, nodeColor=<rgba 1 0 0.9569 1>}
  n117 = ComponentMask(None=WorldPosition().XYZ) {g=true, r=true}
  n118 = ScalarParameter() {defaultValue=512, parameterName=Sand Tiling}
  n119 = TextureSampleParameter2D!(UVs=NamedRerouteDeclaration(None=Divide(A=@n117, B=@n118)) {name=SandTiling, nodeColor=<rgba 1 0 0.9569 1>}, Apply View MipBias) {
   parameterName=RiverBedTexture1_Normal
   samplerSource=SSM_Wrap_WorldGroupSettings
   samplerType=SAMPLERTYPE_Normal
   texture=/Game/Packages/DS_Fortnight/pebbles_N
  }
  NamedRerouteDeclaration(None=@n119.RGB) {name=RiverBedTexture1_Norm, nodeColor=<rgba 1 0 0.9569 1>}
  n120 = NamedRerouteUsage()
  n121 = ScalarParameter() {defaultValue=0.5973, group=WaterFall, parameterName=Waterfall Speed}
  NamedRerouteDeclaration(None=Multiply(A=Multiply(A=@n120, B=@n121), B=Constant2Vector() {g=-1})) {name=FallSpeed, nodeColor=<rgba 1 0 0.9569 1>}
  n122 = NamedRerouteUsage()
  n123 = VertexColor()
  n124 = NamedRerouteDeclaration(None=Multiply(A=@n123.A, B=NamedRerouteUsage()) {desc=Combine vert paint alpha}) {name=FlowmapAlpha-X-VertAlpha, nodeColor=<rgba 1 0 0.9569 1>}
  n125 = LinearInterpolate(
   A=@n122
   B=Multiply(A=@n122, B=Multiply(A=@n123.G, B=@n124)) {desc=Mask Edges}
   Alpha=ScalarParameter() {group=FlowMap, parameterName=Bank Drag}
  )
  n126 = NamedRerouteUsage()
  n127 = ComponentMask(None=Multiply(A=QualitySwitch(Default=@n125, Low=@n126, High=@n125), B=@n58.RGB)) {g=true, r=true}
  NamedRerouteDeclaration(None=@n127) {name=RTT-Flow, nodeColor=<rgba 1 0 0.9569 1>}
  n128 = NamedRerouteUsage()
  n129 = TextureSampleParameter2D(UVs=NamedRerouteUsage()) {
   desc=normal sample - under water and on the shoreline
   parameterName=RiverBedTexture_1
   samplerSource=SSM_Wrap_WorldGroupSettings
   texture=&pebbles
  }
  n130 = Multiply(A=ScalarParameter() {parameterName=Bank Edge Brightness}, B=@n129.RGB)
  n131 = OneMinus(None=@n75)
  n132 = VectorParameter() {parameterName=Sand Color, defaultValue=<rgba 0.5 0.5 1 1>}
  n133 = Multiply(A=@n132.RGB, B=@n132.RGB)
  n134 = SetMaterialAttributes(
   Base Color=Multiply(A=Multiply(A=@n128, B=@n130), B=Multiply(A=@n131, B=@n133))
   Specular=ScalarParameter() {defaultValue=0.1, group=Spec, parameterName=Spec Sand}
   Roughness=ScalarParameter() {group=Roughness, parameterName=Sand Roughness}
   Normal=NamedRerouteUsage()
  ) {attributeSetTypes [
    #3
    #4
    #5
    #6
   ]}
  NamedRerouteDeclaration(None=@n134) {name=SandMat, nodeColor=<rgba 1 0 0.9569 1>}
  NamedRerouteDeclaration(None=OneMinus(None=@n129.A)) {name=SandHeight, nodeColor=<rgba 1 0 0.9569 1>}
  n135 = Multiply(A=@n28.Distance falloff, B=@n28.Distance falloff)
  NamedRerouteDeclaration(None=Saturate(None=Add(A=@n135, B=@n34.Alpha))) {name=HideRocks, nodeColor=<rgba 1 0 0.9569 1>}
  n136 = VectorParameter() {parameterName=Offset}
  n137 = NamedRerouteUsage()
  n138 = ScalarParameter() {group=FlowMap, parameterName=Speed}
  n139 = PreSkinnedPosition()
  n140 = VectorParameter() {group=FixedWaterfallFlow, parameterName=FixedWaterfallFlow-Scale, defaultValue=<rgba 0.006 0.0025 1 1>}
  n141 = ComponentMask(None=VertexInterpolator(VS=Multiply(A=@n139, B=@n140.RGB)).PS) {g=true, r=true}
  n142 = Divide(A=ComponentMask(None=WorldPosition().XYZ) {g=true, r=true}, B=ScalarParameter() {defaultValue=512, parameterName=Size})
  NamedRerouteDeclaration(None=MaterialFunctionCall!(
    Diffuse=TextureObject() {texture=/Game/Packages/Fortnite_UniqueMaterials_1/Textures/FoamHeight}
    Normal Map=TextureObject() {samplerType=SAMPLERTYPE_Normal, texture=/Game/Packages/Fortnite_UniqueMaterials_1/Textures/FoamNormals}
    "Flow Vector Map (see tooltip)"=@n127
    UV texture 2 offset=ComponentMask(None=@n136.RGB) {g=true, r=true}
    "Time (see tooltip)"=Multiply(A=@n137, B=@n138) {desc=graphical bug regarding time forces us to use period}
    UVs=StaticSwitchParameter(True=@n141, False=@n142) {parameterName=FixedWaterfallFlow}
   ) {materialFunction=/Engine/Functions/Engine_MaterialFunctions02/Texturing/FlowMaps_Simple}.Normal) {name=FlowNorm, nodeColor=<rgba 1 0 0.9569 1>}
  n143 = ScalarParameter() {group=FlowMap, parameterName=Pebble Flow Center Mask}
  n144 = TextureSampleParameter2D() {
   parameterName=FlowMap
   samplerSource=SSM_Wrap_WorldGroupSettings
   samplerType=SAMPLERTYPE_LinearColor
   texture=/Game/Packages/Fortnite_UniqueMaterials_1/Textures/T_River_Flowmap_Alpha_Straight_01
  }
  n145 = NamedRerouteDeclaration(None=@n144.A) {name=FlowMapAlpha, nodeColor=<rgba 1 0 0.9569 1>}
  n146 = LinearInterpolate(
   A=Multiply(A=VectorParameter() {group=FlowMap, parameterName=Flow Pebble Normal}.RGB, B=NamedRerouteUsage())
   Alpha=Min(A=Multiply(A=@n143, B=@n145) {constB=1.3})
  ) {constB=0}
  n147 = Constant()
  n148 = Transform(None=ConstantBiasScale(None=@n144.RGB) {bias=-0.5, scale=2}) {desc="!!!", transformSourceType=TRANSFORMSOURCE_World, transformType=TRANSFORM_Local}
  n149 = Constant3Vector() {constant=<rgba 0 -1 0 0>}
  n150 = MaterialFunctionCall() {materialFunction=&TangentBasis}
  n151 = Constant3Vector() {constant=<rgba -1 1 1 0>}
  n152 = StaticSwitchParameter(
   True=Constant3Vector() {constant=<rgba 0 -1 1 1>}
   False=StaticSwitchParameter(True=Transform(None=@n149) {transformSourceType=TRANSFORMSOURCE_Local}, False=Multiply(A=@n150.X U, B=@n151)) {parameterName=UseWaterFall}
  ) {parameterName=FixedWaterfallFlow}
  NamedRerouteDeclaration(None=Add(
    A=QualitySwitch(Default=@n146, Low=@n147, High=@n146)
    B=StaticSwitchParameter(True=@n148, False=@n152) {defaultValue=true, parameterName=Use Flow Texturemap}
   )) {name=FlowType, nodeColor=<rgba 1 0 0.9569 1>}
  ScalarParameter() {defaultValue=1, parameterName=debugwaterfallopacity}
  TextureCoordinate() {uTiling=0.5, vTiling=11}
  TextureCoordinate()
  TextureCoordinate() {uTiling=0.7, vTiling=2.8}
 ]
}
